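# Static UserWidget that displays a "TEST MODE" debug warning banner, named as test content for the Creative EOM screens.
# An Overlay root layers a translucent black background image behind a horizontal box holding an orange warning icon and matching orange text.
# The event graph holds only unconnected PreConstruct, Construct and Tick events, so the widget has no logic or variables.
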WidgetBlueprint /Game/Athena/HUD/Creative/EOMScreens/TestWidgets/WBP_TEST_EOMScreens_DebugWarning_Content
graphs [EventGraph {nodes [
   Event PreConstruct()
   Event Construct()
   Event Tick()
  ]}]
variables []
widgets {
 info {inheritedWidgetCount=0, parentClass=&UserWidget, rootWidgetClass=&Overlay, widgetCount=5}
 widgets [
  Root {widgetClassPath=&Overlay, widget=Root}
  Image_Bg {
   widgetClassPath=&Image
   bIsVariable=true
   parent=Root
   widget=Image_Bg
   properties {colorAndOpacity=<rgba 0 0 0 0.2>}
   slot {class=OverlaySlot, properties {horizontalAlignment=HAlign_Fill, verticalAlignment=VAlign_Fill}}
  }
  Content {widgetClassPath=&HorizontalBox, parent=Root, widget=Content, slot {class=OverlaySlot}}
  Image_Icon {
   widgetClassPath=&Image
   bIsVariable=true
   parent=Content
   widget=Image_Icon
   properties {
    brush {
     bIsDynamicallyLoaded=false
     drawAs=Image
     imageType=NoImage
     mirroring=NoMirror
     resourceName=None
     resourceObject=/Game/Athena/HUD/Bacchus/HUDLayoutTool/HLT_Icons/T-HLT-Warning-64
     tiling=NoTile
     imageSize=<40 40>
     margin=<ltrb 0 0 0 0>
     outlineSettings {
      bUseBrushTransparency=false
      roundingType=HalfHeightRadius
      width=0
      color {colorUseRule=UseColor_Specified, specifiedColor=<rgba 0 0 0 0>}
      cornerRadii=<0 0 0 0>
     }
     tintColor {colorUseRule=UseColor_Specified, specifiedColor=<rgba 1 1 1 1>}
     uVRegion {bIsValid=false, max=<0 0>, min=<0 0>}
    }
    colorAndOpacity=<rgba 1 0.5819 0 1>
   }
   slot {class=HorizontalBoxSlot, properties {horizontalAlignment=HAlign_Center, verticalAlignment=VAlign_Center, padding=<ltrb 10 10 10 10>}}
  }
  Text {
   widgetClassPath=&UEFN_TextBlock
   parent=Content
   widget=Text
   properties {
    text="TEST MODE "
    colorAndOpacity {colorUseRule=UseColor_Specified, specifiedColor=<rgba 1 0.5841 0 1>}
    font {
     bForceMonospaced=false
     bMaterialIsStencil=false
     fontMaterial=null
     fontObject=&BurbankSmall
     letterSpacing=0
     monospacedWidth=1
     size=16
     skewAmount=0
     typefaceFontName=Black
     outlineSettings {
      bApplyOutlineToDropShadows=false
      bMiteredCorners=false
      bSeparateFillAlpha=false
      outlineMaterial=null
      outlineSize=0
      outlineColor=<rgba 0 0 0 1>
     }
    }
   }
   slot {class=HorizontalBoxSlot, properties {verticalAlignment=VAlign_Center, padding=<ltrb 10 0 10 0>}}
  }
 ]
}
