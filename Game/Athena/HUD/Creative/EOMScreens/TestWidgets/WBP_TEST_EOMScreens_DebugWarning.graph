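# Test UMG widget for the Creative EOM screens that places a debug warning in all four screen corners.
# An Overlay root holds four padded NamedSlots aligned to each corner, each filled with its own WBP_TEST_EOMScreens_DebugWarning_Content instance.
# The event graph has only unconnected stub events, so the widget is layout only and any warning visuals or behavior come from the content widget.
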WidgetBlueprint /Game/Athena/HUD/Creative/EOMScreens/TestWidgets/WBP_TEST_EOMScreens_DebugWarning
graphs [EventGraph {nodes [
   Event PreConstruct()
   Event Construct()
   Event Tick()
  ]}]
variables []
widgets {
 info {inheritedWidgetCount=0, parentClass=&UserWidget, rootWidgetClass=&Overlay, widgetCount=9}
 widgets [
  Root {widgetClassPath=&Overlay, widget=Root}
  Slot_TopLeft {
   widgetClassPath=&NamedSlot
   bIsVariable=true
   parent=Root
   widget=Slot_TopLeft
   slot {class=OverlaySlot, properties {padding=<ltrb 20 20 0 0>}}
  }
  DebugWarning_Content_1 {
   widgetClassPath=&/WBP_TEST_EOMScreens_DebugWarning_Content
   bIsVariable=true
   parent=Slot_TopLeft
   widget=DebugWarning_Content_1
   slot {class=PanelSlot}
  }
  Slot_BottomLeft {
   widgetClassPath=&NamedSlot
   bIsVariable=true
   parent=Root
   widget=Slot_BottomLeft
   slot {class=OverlaySlot, properties {verticalAlignment=VAlign_Bottom, padding=<ltrb 20 0 0 20>}}
  }
  DebugWarning_Content_2 {
   widgetClassPath=&WBP_TEST_EOMScreens_DebugWarning_Content
   bIsVariable=true
   parent=Slot_BottomLeft
   widget=DebugWarning_Content_2
   slot {class=PanelSlot}
  }
  Slot_TopRight {
   widgetClassPath=&NamedSlot
   bIsVariable=true
   parent=Root
   widget=Slot_TopRight
   slot {class=OverlaySlot, properties {horizontalAlignment=HAlign_Right, padding=<ltrb 0 20 20 0>}}
  }
  DebugWarning_Content_3 {
   widgetClassPath=&WBP_TEST_EOMScreens_DebugWarning_Content
   bIsVariable=true
   parent=Slot_TopRight
   widget=DebugWarning_Content_3
   slot {class=PanelSlot}
  }
  Slot_BottomRight {
   widgetClassPath=&NamedSlot
   bIsVariable=true
   parent=Root
   widget=Slot_BottomRight
   slot {class=OverlaySlot, properties {horizontalAlignment=HAlign_Right, verticalAlignment=VAlign_Bottom, padding=<ltrb 0 0 20 20>}}
  }
  DebugWarning_Content_4 {
   widgetClassPath=&WBP_TEST_EOMScreens_DebugWarning_Content
   bIsVariable=true
   parent=Slot_BottomRight
   widget=DebugWarning_Content_4
   slot {class=PanelSlot}
  }
 ]
}
